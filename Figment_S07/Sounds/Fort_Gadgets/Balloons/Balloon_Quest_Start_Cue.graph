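# Sound cue for the balloon gadget quest start sound in Figment Season 7.
# A single wave player feeds Balloon_Quest_Start straight to the output at reduced volume, with no randomization or modulation.
# It borrows gadget-activate attenuation and the third-person ranged weapon fire sound class, with a concurrency rule allowing one instance per owner and stopping the oldest.
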SoundCue /Figment_S07/Sounds/Fort_Gadgets/Balloons/Balloon_Quest_Start_Cue
root=WavePlayer_1
nodes [{id=WavePlayer_1, properties {soundWaveAssetPtr=&/Balloon_Quest_Start}}]
properties {
 attenuationSettings=&Fort_GadgetActivate_Att
 maxDistance=4000
 soundClassObject=&Fort_WeaponRanged_Fire_3P_PSM
 volumeMultiplier=0.4
 concurrencySet [/Game/Sounds/Concurrency_Rules/OneMaximum_Owner_StopOldest]
}
